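# Translucent unlit particle sprite material for a bright flash in the level-up UI effects.
# Particle color RGB passes through the MF_FX_EmissiveAdjustments function.
# Particle alpha is multiplied by an inverted depth fade over a fixed distance, so that term is strongest near intersecting scene geometry rather than softened there.
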Material /Game/Effects/Fort_Effects/UI/LevelUp/M_BrightFlash
graph {
 properties {
  bUsedWithParticleSprites=true
  blendMode=BLEND_Translucent
  pixelDepthOffsetMode=PDOM_Legacy
  shadingModel=MSM_Unlit
  usedShadingModels=Unlit
 }
 nodes [
  n1 = ParticleColor()
  MaterialFunctionCall(Emissive=@n1.RGB) {materialFunction=&MF_FX_EmissiveAdjustments}
  n2 = Constant() {r=350}
  Multiply(A=@n1.A, B=OneMinus(None=DepthFade(FadeDistance=@n2)))
 ]
}
