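# Unlit translucent particle sprite material for the LevelUp UI effects, drawing a procedural burst shape without textures.
# Subtracting a wider, softer SphereMask from a smaller, harder one, both centered in UV space, leaves a thin ring.
# The ring is multiplied by particle alpha as a translucency mask.
# Particle color RGB feeds the MF_FX_EmissiveAdjustments function for the glow color.
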Material /Game/Effects/Fort_Effects/UI/LevelUp/M_Burst
graph {
 properties {
  bUsedWithParticleSprites=true
  blendMode=BLEND_Translucent
  pixelDepthOffsetMode=PDOM_Legacy
  shadingModel=MSM_Unlit
  usedShadingModels=Unlit
 }
 nodes [
  n1 = ParticleColor()
  MaterialFunctionCall(Emissive=@n1.RGB) {materialFunction=&MF_FX_EmissiveAdjustments}
  n2 = TextureCoordinate()
  n3 = Constant() {r=0.5}
  n4 = Constant() {r=0.9}
  n5 = TextureCoordinate()
  n6 = Constant() {r=0.5}
  n7 = Constant() {r=0.65}
  n8 = Constant() {r=0.6}
  Multiply(A=Subtract(A=SphereMask(A=@n2, B=@n3, Radius=@n3, Hardness=@n4), B=SphereMask(A=@n5, B=@n6, Radius=@n7, Hardness=@n8)), B=@n1.A)
 ]
}
