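# Translucent unlit mesh-particle material that draws scrolling streaks, apparently for an evolve effect in the level-up UI.
# Particle color passes through an emissive adjustment function for the glow.
# A tiled, vertically panned UV becomes hard stripes via Frac and Floor, multiplied by a sharpened sine pulse along V.
# That mask is multiplied by particle alpha, a panning caustics texture and a depth fade to form what appears to be opacity.
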Material /Game/Effects/Fort_Effects/UI/LevelUp/M_EvolveStreaks
graph {
 properties {bUsedWithMeshParticles=true, blendMode=BLEND_Translucent, pixelDepthOffsetMode=PDOM_Legacy, shadingModel=MSM_Unlit, usedShadingModels=Unlit}
 nodes [
  n1 = ParticleColor()
  MaterialFunctionCall(Emissive=@n1.RGB) {materialFunction=&MF_FX_EmissiveAdjustments}
  n2 = Panner(Coordinate=TextureCoordinate() {uTiling=8, vTiling=2}) {speedY=-2.5}
  n3 = Frac(None=ComponentMask(None=@n2) {r=true})
  n4 = Multiply(A=@n1.A, B=Floor(None=Add(A=@n3) {constB=0.5}))
  n5 = Sine(None=Frac(None=ComponentMask(None=@n2) {g=true}))
  n6 = Power(Base=Saturate(None=Add(A=@n5) {constB=0.5})) {constExponent=5}
  n7 = Add(A=TextureSample(UVs=Panner() {speedX=0.1, speedY=-0.6}) {texture=&T_ShieldCaustics}.RGB) {constB=0.3}
  Multiply(A=Multiply(A=Multiply(A=@n4, B=@n6), B=Saturate(None=@n7)), B=DepthFade(FadeDistance=Constant() {r=6}))
 ]
}
