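# Translucent mesh-particle material for the level-up aura effect, with emissive taken from particle color through MF_FX_EmissiveAdjustments.
# A panning, V-stretched shield caustics texture is multiplied by particle alpha to form the moving translucency mask.
# Two power curves on inverted V multiply into a band that fades at both ends of the mesh's UVs, and a depth fade softens intersections with scene geometry.
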Material /Game/Effects/Fort_Effects/UI/LevelUp/M_LevelUpAura
graph {
 properties {
  bUsedWithMeshParticles=true
  blendMode=BLEND_Translucent
  pixelDepthOffsetMode=PDOM_Legacy
  translucencyLightingMode=TLM_VolumetricPerVertexNonDirectional
  usedShadingModels=Default Lit
 }
 nodes [
  n1 = ParticleColor()
  MaterialFunctionCall(Emissive=@n1.RGB) {materialFunction=&MF_FX_EmissiveAdjustments}
  n2 = TextureSample(UVs=Panner(Coordinate=TextureCoordinate() {vTiling=0.3}) {speedX=-0.3, speedY=1.5}) {texture=&T_ShieldCaustics}
  n3 = ComponentMask(None=TextureCoordinate()) {g=true}
  n4 = OneMinus(None=Power(Base=OneMinus(None=@n3)))
  n5 = Power(Base=OneMinus(None=@n3))
  Multiply(A=Multiply(A=Multiply(A=@n1.A, B=@n2.RGB), B=Multiply(A=@n4, B=@n5)), B=DepthFade(FadeDistance=Constant() {r=8}))
 ]
}
