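# Unlit, opaque material for mesh particles, by its name and folder the burst in the level-up UI effect.
# It adds one to an inverted Fresnel term, so the view-angle falloff brightens the particle color rather than masking it.
# The result multiplies the particle RGB and feeds the shared FX emissive adjustments function.
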Material /Game/Effects/Fort_Effects/UI/LevelUp/M_LevelUpBurst
graph {
 properties {bUsedWithMeshParticles=true, pixelDepthOffsetMode=PDOM_Legacy, shadingModel=MSM_Unlit, usedShadingModels=Unlit}
 nodes [
  n1 = Constant() {r=1}
  n2 = MaterialFunctionCall() {materialFunction=&MF_Fresnel_FX}
  MaterialFunctionCall(Emissive=Multiply(A=ParticleColor().RGB, B=Add(A=@n1, B=@n2.Fresnel Inverted))) {materialFunction=&MF_FX_EmissiveAdjustments}
 ]
}
