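# Translucent unlit distortion material for mesh particles in the level-up UI effect, using index-of-refraction mode.
# A Fresnel function with a fixed exponent is multiplied by particle color alpha to form a blend mask.
# The mask lerps from a neutral value of one to slightly below it, giving an edge-weighted refraction value that each particle's alpha can fade out.
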Material /Game/Effects/Fort_Effects/UI/LevelUp/M_LevelUpDistortion
graph {
 properties {
  bUsedWithMeshParticles=true
  blendMode=BLEND_Translucent
  pixelDepthOffsetMode=PDOM_Legacy
  refractionMethod=RM_IndexOfRefraction
  shadingModel=MSM_Unlit
  usedShadingModels=Unlit
 }
 nodes [
  Constant()
  n1 = Constant() {r=2}
  LinearInterpolate(Alpha=Multiply(
    A=MaterialFunctionCall(Exponent=@n1) {materialFunction=&MF_Fresnel_FX}.Fresnel
    B=ParticleColor().A
   )) {constA=1, constB=0.96}
 ]
}
