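# Unlit, two-sided translucent mesh-particle material, named for the helix effect in the level-up UI.
# Particle color RGB feeds MF_FX_EmissiveAdjustments for the glow.
# Opacity is particle alpha times a procedural soft band that sweeps along the V coordinate over particle relative time.
# The band is a clamped UV-plus-time ramp reshaped by a parabola that peaks mid-ramp and fades to zero at both ends, without textures.
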Material /Game/Effects/Fort_Effects/UI/LevelUp/M_LevelUpHelix
graph {
 properties {
  bUsedWithMeshParticles=true
  blendMode=BLEND_Translucent
  pixelDepthOffsetMode=PDOM_Legacy
  shadingModel=MSM_Unlit
  twoSided=true
  usedShadingModels=Unlit
 }
 nodes [
  n1 = ParticleColor()
  MaterialFunctionCall(Emissive=@n1.RGB) {materialFunction=&MF_FX_EmissiveAdjustments}
  n2 = OneMinus(None=ComponentMask(None=TextureCoordinate()) {g=true})
  n3 = ParticleRelativeTime()
  n4 = Saturate(None=Add(A=Subtract(A=@n2), B=Multiply(A=@n3) {constB=2}))
  n5 = Multiply(A=@n4, B=@n4)
  Multiply(A=@n1.A, B=Add(A=Multiply(A=@n5) {constB=-4}, B=Multiply(A=@n4) {constB=4}))
 ]
}
