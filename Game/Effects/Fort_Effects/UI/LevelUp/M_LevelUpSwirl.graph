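# M_LevelUpSwirl is an unlit translucent material for mesh particles, apparently the swirl in the level-up UI effect.
# Particle color RGB passes through MF_FX_EmissiveAdjustments for the emissive color.
# Particle alpha offsets an inverted V-coordinate gradient and also multiplies it, so lowering alpha retracts the gradient along V while fading it.
# A depth fade multiplies the result to soften intersections.
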Material /Game/Effects/Fort_Effects/UI/LevelUp/M_LevelUpSwirl
graph {
 properties {bUsedWithMeshParticles=true, blendMode=BLEND_Translucent, pixelDepthOffsetMode=PDOM_Legacy, shadingModel=MSM_Unlit, usedShadingModels=Unlit}
 nodes [
  n1 = ParticleColor()
  MaterialFunctionCall(Emissive=@n1.RGB) {materialFunction=&MF_FX_EmissiveAdjustments}
  n2 = ComponentMask(None=Add(A=TextureCoordinate(), B=OneMinus(None=@n1.A)) {constB=0}) {g=true}
  Multiply(A=Multiply(A=@n1.A, B=OneMinus(None=@n2)), B=DepthFade(FadeDistance=Constant() {r=20}))
 ]
}
