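# Translucent unlit, two-sided material for mesh particles, a light cone in the level-up UI effects.
# A power curve on the V texture coordinate forms a gradient that blends OuterCol into InnerCol, passed through MF_FX_EmissiveAdjustments for emissive.
# The same gradient is multiplied by particle alpha and a short depth fade, fading the cone along V and softening intersections with geometry.
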Material /Game/Effects/Fort_Effects/UI/LevelUp/M_LightCone
graph {
 properties {
  bUsedWithMeshParticles=true
  blendMode=BLEND_Translucent
  pixelDepthOffsetMode=PDOM_Legacy
  shadingModel=MSM_Unlit
  twoSided=true
  usedShadingModels=Unlit
 }
 nodes [
  n1 = Power(Base=ComponentMask(None=TextureCoordinate()) {g=true})
  MaterialFunctionCall(Emissive=LinearInterpolate(
    A=VectorParameter() {parameterName=OuterCol, defaultValue=<rgba 0.9504 0.2997 2 0>}.RGB
    B=VectorParameter() {parameterName=InnerCol, defaultValue=<rgba 0.9251 2 0.4105 0>}.RGB
    Alpha=@n1
   )) {materialFunction=&MF_FX_EmissiveAdjustments}
  Multiply(A=Multiply(A=ParticleColor().A, B=@n1), B=DepthFade(FadeDistance=Constant() {r=10}))
 ]
}
